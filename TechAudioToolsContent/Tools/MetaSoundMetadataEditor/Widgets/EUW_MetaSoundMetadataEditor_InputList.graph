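# Editor utility widget that shows the list of MetaSound inputs in the TechAudioTools MetaSound metadata editor.
# It extends TechAudioToolsEditorListView, stacking a search box above an inherited single-selection ListView of input entry widgets.
# Accepting a drop on the list casts the drag payload to an entry viewmodel and clears its dragged flag.
# Reset Search clears and collapses the search bar and sets the input list viewmodel's search text to empty.
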EditorUtilityWidgetBlueprint /TechAudioToolsContent/Tools/MetaSoundMetadataEditor/Widgets/EUW_MetaSoundMetadataEditor_InputList
graphs [
 EventGraph {nodes [
   n1 = "On Item Accept Drop (ListView)"!() -> (OutputDelegate, then, Item:Object Reference, DropZone:EUMGItemDropZone Enum=AboveItem, Operation:Drag Drop Operation Object Reference) {is=title K2Node_ComponentBoundEvent}
   > n2 = Get!(execute:Exec, self:Drag Drop Operation Object Reference=@n1.Operation) -> (then:Exec, else:Exec, Payload:Object Reference)
   > n3 = Cast To TechAudioToolsEntryViewModel!(execute, Object=@n2) -> (then, CastFailed, AsTechAudioTools Entry Viewmodel:Tech Audio Tools Entry Viewmodel Object Reference) {is=title K2Node_DynamicCast}
   > Set bIsBeingDragged!(execute, bIsBeingDragged:bool=false, self:Tech Audio Tools Entry Viewmodel Object Reference=@n3) -> (then, Output_Get:bool)
  ]}
 Reset Search {nodes [
   Reset Search() {is=title K2Node_FunctionEntry}
   > Clear Text(self:Tech Audio Tools Search Box Object Reference=Get InputSearchFilter() -> (InputSearchFilter:Tech Audio Tools Search Box Object Reference))
   > SetVisibility(self=Get Search_Border() -> (Search_Border:Border Object Reference), InVisibility=Collapsed)
   > Set SearchText!(
    execute
    SearchText:Text
    self:Meta Sound Member List Viewmodel Object Reference=Get MetaSoundInputListEditorViewModel() -> (MetaSoundInputListEditorViewModel:Meta Sound Input List Editor Viewmodel Object Reference)
   ) -> (then, Output_Get:Text)
  ]}
]
variables []
widgets {
 info {
  inheritedWidgetCount=1
  parentClass=/Script/TechAudioToolsContentEditor.TechAudioToolsEditorListView
  rootWidgetClass=&VerticalBox
  widgetCount=4
 }
 widgets [
  VerticalBox_1 {widgetClassPath=&VerticalBox, widget=VerticalBox_1}
  Search_Border {
   widgetClassPath=&Border
   bIsVariable=true
   parent=VerticalBox_1
   widget=Search_Border
   properties {brushColor=<rgba 0.0284 0.0284 0.0284 1>, padding=<ltrb 2 0 2 2>}
   slot {class=VerticalBoxSlot}
  }
  InputSearchFilter {
   widgetClassPath=/TechAudioToolsContent/Widgets/Editor/EUW_SearchBox
   bIsVariable=true
   parent=Search_Border
   widget=InputSearchFilter
   slot {class=BorderSlot, properties {padding=<ltrb 2 0 2 2>}}
  }
  ListView {
   widgetClassPath=/Script/Blutility.EditorUtilityListView
   bInherited=true
   bIsVariable=true
   parent=VerticalBox_1
   widget=ListView
   properties {
    bEnableRightClickScrolling=false
    bEnableScrollAnimation=true
    entryWidgetClass=&/EUW_MetaSoundMetadataEditor_InputEntry.EUW_MetaSoundMetadataEditor_InputEntry_C
    numDesignerPreviewEntries=15
    selectionMode=SingleToggle
   }
   slot {class=VerticalBoxSlot, properties {size {sizeRule=Fill, value=1}}}
  }
 ]
}
