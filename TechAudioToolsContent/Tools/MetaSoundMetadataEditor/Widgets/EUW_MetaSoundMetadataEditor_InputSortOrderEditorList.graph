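# Editor utility list widget for reordering MetaSound inputs by sort order in the MetaSound Metadata Editor.
# Drag-drop passes the dragged and target entry indices to the viewmodel's Reorder List Items.
# Reorder calls Apply Reordering, while Cancel rebroadcasts ListItems; both reset sorting to ascending SortOrder, fire a dispatcher and clear selection.
# A helper publishes the selected entry's input editor viewmodel through a field-notify variable.
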EditorUtilityWidgetBlueprint /TechAudioToolsContent/Tools/MetaSoundMetadataEditor/Widgets/EUW_MetaSoundMetadataEditor_InputSortOrderEditorList
graphs [
 EventGraph {nodes [
   n1 = "On Item Accept Drop (ListView)"!() -> (OutputDelegate, then, Item:Object Reference, DropZone:EUMGItemDropZone Enum=AboveItem, Operation:Drag Drop Operation Object Reference) {is=title K2Node_ComponentBoundEvent}
   > Handle Drop Accepted!(execute, self, Item:Object Reference=@n1.Item, Operation:Drag Drop Operation Object Reference=@n1.Operation) -> (then)
   "On Clicked (SortOrderCancel_Button)"() {is=title K2Node_ComponentBoundEvent}
   > Broadcast Field Value Changed!(
    execute
    Object:Object Reference=Get TechAudioToolsListViewModel() -> (TechAudioToolsListViewModel:Tech Audio Tools List Viewmodel Object Reference)
    FieldId:Field Notification Id Structure=(FieldName="ListItems")
   ) -> (then)
   > Reset Sort Options()
   > Call On Cancel Button Clicked!(execute:Exec, self:EUW Meta Sound Metadata Editor Input Sort Order Editor List Object Reference) -> (then:Exec) {is=title K2Node_CallDelegate}
   > Clear Selection(self:List View Object Reference=Get ListView() -> (ListView:List View Object Reference))
   "On Clicked (SortOrderReorder_Button)"() {is=title K2Node_ComponentBoundEvent}
   > Apply Reordering(self:BP Meta Sound Input Sort Order Editor List Viewmodel Object Reference=Get BP_MetaSoundInputSortOrderEditorListViewmodel() -> (BP_MetaSoundInputSortOrderEditorListViewmodel:BP Meta Sound Input Sort Order Editor List Viewmodel Object Reference))
   > Reset Sort Options()
   > Call On Reorder Button Clicked!(execute:Exec, self:EUW Meta Sound Metadata Editor Input Sort Order Editor List Object Reference) -> (then:Exec) {is=title K2Node_CallDelegate}
   > Clear Selection(self=Get ListView())
   Reset Sort Options() {is=title K2Node_CustomEvent}
   n2 = Get BP_MetaSoundInputSortOrderEditorListViewmodel()
   > Set SortMethod!(execute, SortMethod:EMetaSoundMemberEditorSortMethod Enum=SortOrder, self:Meta Sound Input List Editor Viewmodel Object Reference=@n2) -> (then, Output_Get:EMetaSoundMemberEditorSortMethod Enum=Alphabetical)
   > Set SortOrder!(execute, SortOrder:EArraySortOrder Enum=Ascending, self:Meta Sound Member List Viewmodel Object Reference=@n2) -> (then, Output_Get:EArraySortOrder Enum=Ascending)
  ]}
 Handle Drop Accepted {nodes [
   Handle Drop Accepted!() -> (then, Item:Object Reference, Operation:Drag Drop Operation Object Reference) {is=title K2Node_FunctionEntry}
   > n1 = Sequence()
   > n2 = Get!(execute:Exec, self:Drag Drop Operation Object Reference=Get Operation() -> (Operation:Drag Drop Operation Object Reference)) -> (then:Exec, else:Exec, Payload:Object Reference)
   > n3 = Cast To TechAudioToolsEntryViewModel!(execute, Object=@n2) -> (then, CastFailed, AsTechAudioTools Entry Viewmodel:Tech Audio Tools Entry Viewmodel Object Reference) {is=title K2Node_DynamicCast}
   > Set bIsBeingDragged!(execute, bIsBeingDragged:bool=false, self:Tech Audio Tools Entry Viewmodel Object Reference=@n3) -> (then, Output_Get:bool)
   n4 = Sequence(execute=@n1.then_1)
   > n5 = Get(self=Get Operation())
   > n6 = Cast To TechAudioToolsEntryViewModel(Object=@n5)
   > Set Old Index(Old Index:int=Get EntryIndex!(self:Tech Audio Tools Entry Viewmodel Object Reference=@n6) -> (EntryIndex:int=0)) -> (Output_Get:int=0)
   n7 = Sequence(execute=@n4.then_1)
   > n8 = Cast To TechAudioToolsEntryViewModel(Object=Get Item() -> (Item:Object Reference))
   > Set New Index(New Index:int=Get EntryIndex(self=@n8) -> (EntryIndex=0)) -> (Output_Get:int=0)
   Sequence(execute=@n7.then_1)
   > Reorder List Items!(
    execute
    self:BP Meta Sound Input Sort Order Editor List Viewmodel Object Reference=Get BP_MetaSoundInputSortOrderEditorListViewmodel()
    Old Index:int=Get Old Index() -> (Old Index:int=0)
    New Index:int=Get New Index() -> (New Index:int=0)
   ) -> (then)
  ]}
 On Cancel Button Clicked {nodes [On Cancel Button Clicked() {is=title K2Node_FunctionEntry}]}
 On Reorder Button Clicked {nodes [On Reorder Button Clicked() {is=title K2Node_FunctionEntry}]}
 Set Selected Input Editor Viewmodel {nodes [
   n1 = Set Selected Input Editor Viewmodel!() -> (then, Selected Entry Viewmodel:Tech Audio Tools Entry Viewmodel Object Reference) {is=title K2Node_FunctionEntry}
   > n2 = Is Valid(InputObject=@n1)
   Set with Broadcast Selected MetaSound Input Editor Viewmodel!(execute=@n2.Is Not Valid, Selected MetaSound Input Editor Viewmodel:Meta Sound Input Editor Viewmodel Object Reference) -> (then, Output_Get:Meta Sound Input Editor Viewmodel Object Reference)
   Set with Broadcast Selected MetaSound Input Editor Viewmodel(
    execute=@n2.Is Valid
    Selected MetaSound Input Editor Viewmodel=Cast To MetaSoundInputEditorViewModel!(Object=Get EntryData!(self:Tech Audio Tools Entry Viewmodel Object Reference=@n1) -> (EntryData:Object Reference)) -> (AsMetaSound Input Editor Viewmodel:Meta Sound Input Editor Viewmodel Object Reference, bSuccess:bool=~) {is=title K2Node_DynamicCast}
   )
  ]}
]
variables [
 On Cancel Button Clicked {default=()}
 On Reorder Button Clicked {default=()}
 Selected MetaSound Input Editor Viewmodel {type=MetaSoundInputEditorViewModel, default=null}
]
widgets {
 info {
  inheritedWidgetCount=1
  namedSlotCount=2
  parentClass=/Script/TechAudioToolsContentEditor.TechAudioToolsEditorListView
  rootWidgetClass=&VerticalBox
  widgetCount=10
 }
 widgets [
  VerticalBox_1 {widgetClassPath=&VerticalBox, widget=VerticalBox_1}
  Sort_Border {
   widgetClassPath=&Border
   bIsVariable=true
   parent=VerticalBox_1
   widget=Sort_Border
   properties {brushColor=<rgba 0.0176 0.0176 0.0176 1>, padding=<ltrb 6 4 6 4>}
   slot {class=VerticalBoxSlot}
  }
  HorizontalBox_0 {
   widgetClassPath=&HorizontalBox
   parent=Sort_Border
   widget=HorizontalBox_0
   slot {class=BorderSlot, properties {padding=<ltrb 6 4 6 4>}}
  }
  TextBlock_455 {
   widgetClassPath=&TextBlock
   parent=HorizontalBox_0
   widget=TextBlock_455
   properties {
    text="Editing Sort Order "
    colorAndOpacity {colorUseRule=UseColor_Specified, specifiedColor=<rgba 0.5776 0.5776 0.5776 1>}
    font {
     bForceMonospaced=false
     bMaterialIsStencil=false
     fontMaterial=null
     fontObject=&Roboto
     letterSpacing=0
     monospacedWidth=1
     size=9
     skewAmount=0
     typefaceFontName=Regular
     outlineSettings {
      bApplyOutlineToDropShadows=false
      bMiteredCorners=false
      bSeparateFillAlpha=false
      outlineMaterial=null
      outlineSize=0
      outlineColor=<rgba 0 0 0 1>
     }
    }
   }
   slot {class=HorizontalBoxSlot, properties {verticalAlignment=VAlign_Center}}
  }
  SortOptions {
   widgetClassPath=/TechAudioToolsContent/Widgets/Editor/Lists/MetaSoundMemberLists/EUW_MetaSoundInputListSortMenu
   bIsVariable=true
   parent=HorizontalBox_0
   widget=SortOptions
   slot {class=HorizontalBoxSlot, properties {horizontalAlignment=HAlign_Left, verticalAlignment=VAlign_Center, size {sizeRule=Fill, value=1}}}
  }
  SortOrderCancel_Button {
   widgetClassPath=/TechAudioToolsContent/Widgets/Editor/Buttons/EUW_Button_Secondary
   bIsVariable=true
   parent=HorizontalBox_0
   widget=SortOrderCancel_Button
   slot {class=HorizontalBoxSlot, properties {padding=<ltrb 0 0 4 0>}}
  }
  TextBlock {
   widgetClassPath=&TextBlock
   namedSlotHost=SortOrderCancel_Button
   widget=TextBlock
   properties {
    text=Cancel
    font {
     bForceMonospaced=false
     bMaterialIsStencil=false
     fontMaterial=null
     fontObject=&Roboto
     letterSpacing=0
     monospacedWidth=1
     size=9
     skewAmount=0
     typefaceFontName=Regular
     outlineSettings {
      bApplyOutlineToDropShadows=false
      bMiteredCorners=false
      bSeparateFillAlpha=false
      outlineMaterial=null
      outlineSize=0
      outlineColor=<rgba 0 0 0 1>
     }
    }
   }
  }
  SortOrderReorder_Button {
   widgetClassPath=/TechAudioToolsContent/Widgets/Editor/Buttons/EUW_Button_Primary
   bIsVariable=true
   parent=HorizontalBox_0
   widget=SortOrderReorder_Button
   slot {class=HorizontalBoxSlot}
  }
  TextBlock_293 {
   widgetClassPath=&TextBlock
   namedSlotHost=SortOrderReorder_Button
   widget=TextBlock_293
   properties {
    text=Reorder
    font {
     bForceMonospaced=false
     bMaterialIsStencil=false
     fontMaterial=null
     fontObject=&Roboto
     letterSpacing=0
     monospacedWidth=1
     size=9
     skewAmount=0
     typefaceFontName=Regular
     outlineSettings {
      bApplyOutlineToDropShadows=false
      bMiteredCorners=false
      bSeparateFillAlpha=false
      outlineMaterial=null
      outlineSize=0
      outlineColor=<rgba 0 0 0 1>
     }
    }
   }
  }
  ListView {
   widgetClassPath=/Script/Blutility.EditorUtilityListView
   bInherited=true
   bIsVariable=true
   parent=VerticalBox_1
   widget=ListView
   properties {
    bAllowDragDrop=true
    bEnableRightClickScrolling=false
    bEnableScrollAnimation=true
    dragDropOperationClass=/Script/UMG.DragDropOperation
    dragDropVisualEntryClass=&/EUW_MetaSoundMetadataEditor_InputEntry.EUW_MetaSoundMetadataEditor_InputEntry_C
    dragDropVisualPivot=CenterLeft
    entryWidgetClass=&/EUW_MetaSoundMetadataEditor_InputSortOrderEditorEntry.EUW_MetaSoundMetadataEditor_InputSortOrderEditorEntry_C
    numDesignerPreviewEntries=15
    selectionMode=SingleToggle
   }
   slot {class=VerticalBoxSlot, properties {size {sizeRule=Fill, value=1}}}
  }
 ]
}
