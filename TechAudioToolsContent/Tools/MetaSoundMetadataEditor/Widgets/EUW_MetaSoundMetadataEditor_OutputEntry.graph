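# Editor utility list-entry widget for one MetaSound output in the MetaSound Metadata Editor tool.
# On list item set, it casts the item to a MetaSoundOutputEditorViewModel and assigns it as the viewmodel.
# Its row shows name, display name, description, advanced-pin and interface icons, and sort order.
# Helper functions derive icon visibility, hover/selection/drag color, and current or edited sort index text, apparently for view bindings.
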EditorUtilityWidgetBlueprint /TechAudioToolsContent/Tools/MetaSoundMetadataEditor/Widgets/EUW_MetaSoundMetadataEditor_OutputEntry
graphs [
 EventGraph {nodes [
   n1 = Event OnListItemObjectSet()
   > n2 = Cast To MetaSoundOutputEditorViewModel!(execute, Object=@n1.ListItemObject) -> (then, CastFailed, AsMetaSound Output Editor Viewmodel:Meta Sound Output Editor Viewmodel Object Reference) {is=title K2Node_DynamicCast}
   > SetMetaSoundOutputEditorViewModel!(execute, self, Viewmodel:Meta Sound Output Editor Viewmodel Object Reference=@n2) -> (then)
  ]}
 Get Description Icon Visibility {nodes [
   n1 = Get Description Icon Visibility!() -> (then, Description:"Text (by ref)") {is=title K2Node_FunctionEntry}
   > Return Node!(execute, Visibility:ESlateVisibility Enum=Select(Option 0=Visible, Option 1=Collapsed, Index=TextIsEmpty(InText=@n1)))
  ]}
 Get Entry Icon Color {nodes [
   n1 = Get Entry Icon Color!() -> (
    then
    bIsHovered:bool=~
    bIsSelected:bool=~
    bIsDraggedOver:bool=~
    bIsBeingDragged:bool=~
    NormalTextColor:Linear Color Structure
    HighlightedTextColor:Linear Color Structure
   ) {is=title K2Node_FunctionEntry}
   > Return Node!(
    execute
    Specified Color:Linear Color Structure=Break Slate Color!(SlateColor:"Slate Color Structure (by ref)"=GetEntryTextColor!(
      bIsHovered:bool=@n1
      bIsSelected:bool=@n1.bIsSelected
      bIsDraggedOver:bool=@n1.bIsDraggedOver
      bIsBeingDragged:bool=@n1.bIsBeingDragged
      NormalTextColor:Linear Color Structure=@n1.NormalTextColor
      HighlightedTextColor:Linear Color Structure=@n1.HighlightedTextColor
     ) -> (ReturnValue:Slate Color Structure)) -> (SpecifiedColor:Linear Color Structure, ColorUseRule:ESlateColorStylingMode Enum) {is=title K2Node_BreakStruct}
   )
  ]}
 Get Interface Icon Visibility {nodes [
   n1 = Get Interface Icon Visibility!() -> (then, MemberName:"Name (by ref)") {is=title K2Node_FunctionEntry}
   > Return Node!(
    execute
    Visibility:ESlateVisibility Enum=Select(
     Option 0=Collapsed
     Option 1=Visible
     Index=Is MetaSound Interface Member!(MemberName:"Name (by ref)"=@n1, bInvert:bool=false) -> (ReturnValue:bool)
    )
   )
  ]}
 Get Sort Order Index as Text {nodes [
   n1 = Get Sort Order Index as Text!() -> (then, Is Editing Sort Order:bool=~, Sort Order Index:int, New Sort Order Index:int) {is=title K2Node_FunctionEntry}
   > Return Node!(
    execute
    Text:Text=SelectText(A="To Text (Integer)"(Value=@n1.New Sort Order Index), B="To Text (Integer)"(Value=@n1.Sort Order Index), bPickA=@n1)
   )
  ]}
]
variables []
widgets {
 info {parentClass=/Script/TechAudioToolsContentEditor.TechAudioToolsEditorEntry, rootWidgetClass=&VerticalBox, widgetCount=11}
 widgets [
  VerticalBox {widgetClassPath=&VerticalBox, widget=VerticalBox}
  Background {
   widgetClassPath=&Border
   bIsVariable=true
   parent=VerticalBox
   widget=Background
   properties {verticalAlignment=VAlign_Center, brushColor=<rgba 0 0 0 0.1>, padding=<ltrb 0 4 4 4>}
   slot {class=VerticalBoxSlot}
  }
  HorizontalBox_31 {
   widgetClassPath=&HorizontalBox
   parent=Background
   widget=HorizontalBox_31
   slot {class=BorderSlot, properties {verticalAlignment=VAlign_Center, padding=<ltrb 0 4 4 4>}}
  }
  TechAudioToolsEditorSlateIcon_112 {
   widgetClassPath=/Script/TechAudioToolsContentEditor.TechAudioToolsEditorSlateIcon
   parent=HorizontalBox_31
   widget=TechAudioToolsEditorSlateIcon_112
   properties {fullIconPath=EditorStyle.Icons.BulletPoint, imageColorAndOpacity=<rgba 0.5271 0.5271 0.5271 1>}
   slot {class=HorizontalBoxSlot, properties {padding=<ltrb 4 0 4 0>}}
  }
  OutputName_Text {
   widgetClassPath=&CommonTextBlock
   parent=HorizontalBox_31
   widget=OutputName_Text
   properties {
    text=Output Name
    colorAndOpacity {colorUseRule=UseColor_Specified, specifiedColor=<rgba 0.5271 0.5271 0.5271 1>}
    font {
     bForceMonospaced=false
     bMaterialIsStencil=false
     fontMaterial=null
     fontObject=&Roboto
     letterSpacing=0
     monospacedWidth=1
     size=9
     skewAmount=0
     typefaceFontName=Regular
     outlineSettings {
      bApplyOutlineToDropShadows=false
      bMiteredCorners=false
      bSeparateFillAlpha=false
      outlineMaterial=null
      outlineSize=0
      outlineColor=<rgba 0 0 0 1>
     }
    }
   }
   slot {class=HorizontalBoxSlot}
  }
  DisplayName_Text {
   widgetClassPath=&CommonTextBlock
   parent=HorizontalBox_31
   widget=DisplayName_Text
   properties {
    text=Display Name
    visibility=Visible
    colorAndOpacity {colorUseRule=UseColor_Specified, specifiedColor=<rgba 0.1046 0.1046 0.1046 1>}
    font {
     bForceMonospaced=false
     bMaterialIsStencil=false
     fontMaterial=null
     fontObject=&Roboto
     letterSpacing=0
     monospacedWidth=1
     size=9
     skewAmount=0
     typefaceFontName=Regular
     outlineSettings {
      bApplyOutlineToDropShadows=false
      bMiteredCorners=false
      bSeparateFillAlpha=false
      outlineMaterial=null
      outlineSize=0
      outlineColor=<rgba 0 0 0 1>
     }
    }
   }
   slot {class=HorizontalBoxSlot, properties {padding=<ltrb 4 0 6 0>, size {sizeRule=Fill, value=1}}}
  }
  DescriptionIcon {
   widgetClassPath=&TechAudioToolsEditorSlateIcon
   parent=HorizontalBox_31
   widget=DescriptionIcon
   properties {fullIconPath=EditorStyle.Icons.Details, imageColorAndOpacity=<rgba 0.5271 0.5271 0.5271 1>}
   slot {class=HorizontalBoxSlot, properties {horizontalAlignment=HAlign_Center, verticalAlignment=VAlign_Center, padding=<ltrb 0 0 4 0>}}
  }
  AdvancedPin_Icon {
   widgetClassPath=&TechAudioToolsEditorSlateIcon
   parent=HorizontalBox_31
   widget=AdvancedPin_Icon
   properties {fullIconPath=EditorStyle.Icons.Toolbar.Settings, toolTipText=Is Advanced Pin, imageColorAndOpacity=<rgba 0.5271 0.5271 0.5271 1>}
   slot {class=HorizontalBoxSlot, properties {horizontalAlignment=HAlign_Center, verticalAlignment=VAlign_Center, padding=<ltrb 0 0 4 0>}}
  }
  InterfaceIcon {
   widgetClassPath=&TechAudioToolsEditorSlateIcon
   parent=HorizontalBox_31
   widget=InterfaceIcon
   properties {
    fullIconPath=EditorStyle.ClassIcon.Interface
    toolTipText=This member belongs to an interface.
    imageColorAndOpacity=<rgba 0.5271 0.5271 0.5271 1>
   }
   slot {class=HorizontalBoxSlot, properties {horizontalAlignment=HAlign_Center, verticalAlignment=VAlign_Center, padding=<ltrb 0 0 4 0>}}
  }
  SizeBox_121 {
   widgetClassPath=&SizeBox
   parent=HorizontalBox_31
   widget=SizeBox_121
   properties {bOverride_WidthOverride=true, widthOverride=20}
   slot {class=HorizontalBoxSlot, properties {verticalAlignment=VAlign_Center}}
  }
  SortOrderIndex_Text {
   widgetClassPath=&TextBlock
   parent=SizeBox_121
   widget=SortOrderIndex_Text
   properties {
    text=0
    toolTipText=Sort Order
    colorAndOpacity {colorUseRule=UseColor_Specified, specifiedColor=<rgba 0.5271 0.5271 0.5271 1>}
    font {
     bForceMonospaced=false
     bMaterialIsStencil=false
     fontMaterial=null
     fontObject=&Roboto
     letterSpacing=0
     monospacedWidth=1
     size=9
     skewAmount=0
     typefaceFontName=Regular
     outlineSettings {
      bApplyOutlineToDropShadows=false
      bMiteredCorners=false
      bSeparateFillAlpha=false
      outlineMaterial=null
      outlineSize=0
      outlineColor=<rgba 0 0 0 1>
     }
    }
   }
   slot {class=SizeBoxSlot, properties {horizontalAlignment=HAlign_Center}}
  }
 ]
}
