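# Editor utility widget that shows the searchable list of outputs in the MetaSound Metadata Editor tool.
# It extends TechAudioToolsEditorListView, adding a search box above the inherited single-toggle ListView of output entry widgets.
# Selection unwraps the entry viewmodel into a field-notify output editor viewmodel, or broadcasts null when invalid.
# Reset Search clears and collapses the search and empties the list viewmodel's SearchText; accepted drops clear the payload's dragged flag.
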EditorUtilityWidgetBlueprint /TechAudioToolsContent/Tools/MetaSoundMetadataEditor/Widgets/EUW_MetaSoundMetadataEditor_OutputList
graphs [
 EventGraph {nodes [
   n1 = "On Item Accept Drop (ListView)"!() -> (OutputDelegate, then, Item:Object Reference, DropZone:EUMGItemDropZone Enum=AboveItem, Operation:Drag Drop Operation Object Reference) {is=title K2Node_ComponentBoundEvent}
   > n2 = Get!(execute:Exec, self:Drag Drop Operation Object Reference=@n1.Operation) -> (then:Exec, else:Exec, Payload:Object Reference)
   > n3 = Cast To TechAudioToolsEntryViewModel!(execute, Object=@n2) -> (then, CastFailed, AsTechAudioTools Entry Viewmodel:Tech Audio Tools Entry Viewmodel Object Reference) {is=title K2Node_DynamicCast}
   > Set bIsBeingDragged!(execute, bIsBeingDragged:bool=false, self:Tech Audio Tools Entry Viewmodel Object Reference=@n3) -> (then, Output_Get:bool)
  ]}
 Reset Search {nodes [
   Reset Search() {is=title K2Node_FunctionEntry}
   > Clear Text(self:Tech Audio Tools Search Box Object Reference=Get OutputSearchFilter() -> (OutputSearchFilter:Tech Audio Tools Search Box Object Reference))
   > SetVisibility(self=Get Search_Border() -> (Search_Border:Border Object Reference), InVisibility=Collapsed)
   > Set SearchText!(
    execute
    SearchText:Text
    self:Meta Sound Member List Viewmodel Object Reference=Get MetaSoundOutputListEditorViewModel() -> (MetaSoundOutputListEditorViewModel:Meta Sound Output List Editor Viewmodel Object Reference)
   ) -> (then, Output_Get:Text)
  ]}
 Set Selected Output Editor Viewmodel {nodes [
   n1 = Set Selected Output Editor Viewmodel!() -> (then, Selected Entry Viewmodel:Tech Audio Tools Entry Viewmodel Object Reference) {is=title K2Node_FunctionEntry}
   > n2 = Is Valid(InputObject=@n1)
   Set with Broadcast Selected MetaSound Output Editor Viewmodel!(execute=@n2.Is Not Valid, Selected MetaSound Output Editor Viewmodel:Meta Sound Output Editor Viewmodel Object Reference) -> (then, Output_Get:Meta Sound Output Editor Viewmodel Object Reference)
   Set with Broadcast Selected MetaSound Output Editor Viewmodel(
    execute=@n2.Is Valid
    Selected MetaSound Output Editor Viewmodel=Cast To MetaSoundOutputEditorViewModel!(Object=Get EntryData!(self:Tech Audio Tools Entry Viewmodel Object Reference=@n1) -> (EntryData:Object Reference)) -> (AsMetaSound Output Editor Viewmodel:Meta Sound Output Editor Viewmodel Object Reference, bSuccess:bool=~) {is=title K2Node_DynamicCast}
   )
  ]}
]
variables [Selected MetaSound Output Editor Viewmodel {type=MetaSoundOutputEditorViewModel, default=null}]
widgets {
 info {
  inheritedWidgetCount=1
  parentClass=/Script/TechAudioToolsContentEditor.TechAudioToolsEditorListView
  rootWidgetClass=&VerticalBox
  widgetCount=4
 }
 widgets [
  VerticalBox_1 {widgetClassPath=&VerticalBox, widget=VerticalBox_1}
  Search_Border {
   widgetClassPath=&Border
   bIsVariable=true
   parent=VerticalBox_1
   widget=Search_Border
   properties {brushColor=<rgba 0.0284 0.0284 0.0284 1>, padding=<ltrb 2 0 2 2>}
   slot {class=VerticalBoxSlot}
  }
  OutputSearchFilter {
   widgetClassPath=/TechAudioToolsContent/Widgets/Editor/EUW_SearchBox
   bIsVariable=true
   parent=Search_Border
   widget=OutputSearchFilter
   slot {class=BorderSlot, properties {padding=<ltrb 2 0 2 2>}}
  }
  ListView {
   widgetClassPath=/Script/Blutility.EditorUtilityListView
   bInherited=true
   bIsVariable=true
   parent=VerticalBox_1
   widget=ListView
   properties {
    bEnableRightClickScrolling=false
    bEnableScrollAnimation=true
    entryWidgetClass=&/EUW_MetaSoundMetadataEditor_OutputEntry.EUW_MetaSoundMetadataEditor_OutputEntry_C
    numDesignerPreviewEntries=15
    selectionMode=SingleToggle
   }
   slot {class=VerticalBoxSlot, properties {size {sizeRule=Fill, value=1}}}
  }
 ]
}
