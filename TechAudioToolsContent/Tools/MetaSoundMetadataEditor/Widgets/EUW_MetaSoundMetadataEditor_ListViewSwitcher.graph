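# Editor utility widget that switches the MetaSound Metadata Editor between input and output metadata lists and their sort-order editors.
# Header toggles set the field-notify Widget State enum, which helper functions map to switcher index and toggle states.
# Changing tabs or sort mode resets search filters.
# Reorder fires a dispatcher and Cancel resets sort lists; both return to viewing, using FieldNotify broadcasts to refresh list viewmodels.
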EditorUtilityWidgetBlueprint /TechAudioToolsContent/Tools/MetaSoundMetadataEditor/Widgets/EUW_MetaSoundMetadataEditor_ListViewSwitcher
graphs [
 EventGraph {nodes [
   n1 = "On Reorder Button Clicked (OutputSortOrderEditor_ListView)"() {is=title K2Node_ComponentBoundEvent}
   Call On Reorder Clicked!(execute:Exec=[@n1.then, @n17.then], self:EUW Meta Sound Metadata Editor List View Switcher Object Reference) -> (then:Exec) {is=title K2Node_CallDelegate}
   n2 = Get Inputs_ToggleButton() -> (Inputs_ToggleButton:Tech Audio Tools Toggle Button Content Slot Object Reference)
   > n3 = Set with Broadcast Widget State!(
    execute
    Widget State:EMetaSoundMetadataEditorState Enum=Select(
     Option 0:EMetaSoundMetadataEditorState Enum=NewEnumerator2
     Option 1:EMetaSoundMetadataEditorState Enum=NewEnumerator0
     Index=Is Checked!(self:Tech Audio Tools Toggle Button Content Slot Object Reference=@n2) -> (Return Value:bool)
    ) -> (ReturnValue:EMetaSoundMetadataEditorState Enum)
   ) -> (then, Output_Get:EMetaSoundMetadataEditorState Enum=NewEnumerator0)
   > n4 = Switch on EMetaSoundMetadataEditorState!(execute, Selection:EMetaSoundMetadataEditorState Enum=@n3) -> (NewEnumerator0:Exec, NewEnumerator1:Exec, NewEnumerator2:Exec, NewEnumerator3:Exec) {is=title K2Node_SwitchEnum}
   > Broadcast Field Value Changed!(
    execute
    Object:Object Reference=Get MetaSoundInputListEditorViewModel() -> (MetaSoundInputListEditorViewModel:Meta Sound Input List Editor Viewmodel Object Reference)
    FieldId:Field Notification Id Structure=(FieldName="GetFilteredAndSortedEntryViewModels")
   ) -> (then)
   Broadcast Field Value Changed(
    execute=@n4.NewEnumerator2
    Object=Get MetaSoundOutputListEditorViewModel() -> (MetaSoundOutputListEditorViewModel:Meta Sound Output List Editor Viewmodel Object Reference)
    FieldId=(FieldName="GetFilteredAndSortedEntryViewModels")
   )
   n5 = "On Checked State Changed (Search_ToggleButton)"!() -> (OutputDelegate, then, Is Checked:bool) {is=title K2Node_ComponentBoundEvent}
   > Set Checked State!(
    execute
    self:EUW Meta Sound Member List Search Toggle Button Object Reference=Get Search_ToggleButton() -> (Search_ToggleButton:EUW Meta Sound Member List Search Toggle Button Object Reference)
    Is Checked:bool=@n5.Is Checked
   ) -> (then) {is=title K2Node_CallFunction}
   n6 = Get Inputs_ToggleButton()
   > Set with Broadcast Widget State(Widget State=Select(Option 0=NewEnumerator2, Option 1=NewEnumerator0, Index=Is Checked(self=@n6))) -> (Output_Get=NewEnumerator0)
   > n7 = Branch(Condition=@n5.Is Checked)
   Reset Search Filters(execute=@n7.else)
   n8 = "On Cancel Button Clicked (InputSortOrderEditor_ListView)"() {is=title K2Node_ComponentBoundEvent}
   n9 = Get Inputs_ToggleButton()
   Set with Broadcast Widget State(execute=[@n8.then, @n16.then], Widget State=Select(Option 0=NewEnumerator2, Option 1=NewEnumerator0, Index=Is Checked(self=@n9))) -> (Output_Get=NewEnumerator0)
   > Reset Sort Order Lists()
   n10 = "On Checked State Changed (SortOrderEditor_ToggleButton)"!() -> (OutputDelegate, then, Is Checked:bool) {is=title K2Node_ComponentBoundEvent}
   > n11 = Branch(Condition=@n10.Is Checked)
   n12 = Get Inputs_ToggleButton()
   > n13 = Set with Broadcast Widget State(Widget State=Select(Option 0=NewEnumerator3, Option 1=NewEnumerator1, Index=Is Checked(self=@n12))) -> (Output_Get=NewEnumerator0)
   Reset Search Filters(execute=[@n15.then, @n13.then])
   n14 = Get Inputs_ToggleButton()
   n15 = Set with Broadcast Widget State(execute=@n11.else, Widget State=Select(Option 0=NewEnumerator2, Option 1=NewEnumerator0, Index=Is Checked(self=@n14))) -> (Output_Get=NewEnumerator0)
   n16 = "On Cancel Button Clicked (OutputSortOrderEditor_ListView)"() {is=title K2Node_ComponentBoundEvent}
   n17 = "On Reorder Button Clicked (InputSortOrderEditor_ListView)"() {is=title K2Node_ComponentBoundEvent}
   n18 = "On Checked State Changed (Inputs_ToggleButton)"!() -> (OutputDelegate, then, Is Checked:bool) {is=title K2Node_ComponentBoundEvent}
   > n19 = Branch(Condition=@n18.Is Checked)
   > Branch(Condition="Not Equal (Enum)"!(
     A:EMetaSoundMetadataEditorState Enum=Get Widget State() -> (Widget State:EMetaSoundMetadataEditorState Enum=NewEnumerator0)
     B:EMetaSoundMetadataEditorState Enum=NewEnumerator1
    ) -> (ReturnValue:bool))
   > Set with Broadcast Widget State(Widget State=NewEnumerator0) -> (Output_Get=NewEnumerator0)
   > Reset Search Filters()
   Set Checked State(execute=@n19.else, self:Tech Audio Tools Toggle Button Content Slot Object Reference=Get Inputs_ToggleButton(), Is Checked=true)
   n20 = "On Checked State Changed (Outputs_ToggleButton)"!() -> (OutputDelegate, then, Is Checked:bool) {is=title K2Node_ComponentBoundEvent}
   > n21 = Branch(Condition=@n20.Is Checked)
   > Branch(Condition="Not Equal (Enum)"(A=Get Widget State() -> (Widget State=NewEnumerator0), B=NewEnumerator3))
   > Set with Broadcast Widget State(Widget State=NewEnumerator2) -> (Output_Get=NewEnumerator0)
   > Reset Search Filters()
   Set Checked State(
    execute=@n21.else
    self=Get Outputs_ToggleButton() -> (Outputs_ToggleButton:Tech Audio Tools Toggle Button Content Slot Object Reference)
    Is Checked=true
   )
   Reset Search Filters() {is=title K2Node_CustomEvent}
   > Set Checked State(self:EUW Meta Sound Member List Search Toggle Button Object Reference=Get Search_ToggleButton(), Is Checked=false)
   > Reset Search(self:EUW Meta Sound Metadata Editor Input List Object Reference=Get InputMetadata_ListView() -> (InputMetadata_ListView:EUW Meta Sound Metadata Editor Input List Object Reference))
   > Reset Search(self:EUW Meta Sound Metadata Editor Output List Object Reference=Get OutputMetadata_ListView() -> (OutputMetadata_ListView:EUW Meta Sound Metadata Editor Output List Object Reference))
   Reset Sort Order Lists() {is=title K2Node_CustomEvent}
   > Broadcast Field Value Changed(
    Object=Get BP_MetaSoundInputSortOrderEditorListViewmodel!(self:EUW Meta Sound Metadata Editor Input Sort Order Editor List Object Reference=Get InputSortOrderEditor_ListView() -> (InputSortOrderEditor_ListView:EUW Meta Sound Metadata Editor Input Sort Order Editor List Object Reference)) -> (BP_MetaSoundInputSortOrderEditorListViewmodel:BP Meta Sound Input Sort Order Editor List Viewmodel Object Reference)
    FieldId=(FieldName="GetFilteredAndSortedEntryViewModels")
   )
   > Broadcast Field Value Changed(
    Object=Get BP_MetaSoundOutputSortOrderEditorListViewmodel!(self:EUW Meta Sound Metadata Editor Output Sort Order Editor List Object Reference=Get OutputSortOrderEditor_ListView() -> (OutputSortOrderEditor_ListView:EUW Meta Sound Metadata Editor Output Sort Order Editor List Object Reference)) -> (BP_MetaSoundOutputSortOrderEditorListViewmodel:BP Meta Sound Output Sort Order Editor List Viewmodel Object Reference)
    FieldId=(FieldName="GetFilteredAndSortedEntryViewModels")
   )
  ]}
 Get Input Toggle Checked State {nodes [
   Get Input Toggle Checked State() {is=title K2Node_FunctionEntry}
   > Return Node!(
    execute
    ReturnValue:bool=Select!(
     NewEnumerator0:bool=true
     NewEnumerator1:bool=true
     NewEnumerator2:bool=false
     NewEnumerator3:bool=false
     Index:EMetaSoundMetadataEditorState Enum=Get Widget State() -> (Widget State=NewEnumerator0)
    ) -> (ReturnValue:bool)
   )
  ]}
 Get Output Toggle Checked State {nodes [
   Get Output Toggle Checked State() {is=title K2Node_FunctionEntry}
   > Return Node(ReturnValue=Select(
     NewEnumerator0=false
     NewEnumerator1=false
     NewEnumerator2=true
     NewEnumerator3=true
     Index=Get Widget State() -> (Widget State=NewEnumerator0)
    ))
  ]}
 Get Selected List Index {nodes [
   n1 = Get Selected List Index!() -> (then, State:EMetaSoundMetadataEditorState Enum=NewEnumerator0) {is=title K2Node_FunctionEntry}
   > Return Node!(
    execute
    Index:int=Select(NewEnumerator0:int=0, NewEnumerator1:int=1, NewEnumerator2:int=2, NewEnumerator3:int=3, Index=@n1) -> (ReturnValue:int)
   )
  ]}
 Get Sort Order Toggle Checked State {nodes [
   Get Sort Order Toggle Checked State() {is=title K2Node_FunctionEntry}
   > Return Node!(
    execute
    ReturnValue:bool=Select(
     NewEnumerator0:bool=false
     NewEnumerator1:bool=true
     NewEnumerator2:bool=false
     NewEnumerator3:bool=false
     Index=Get Widget State() -> (Widget State=NewEnumerator0)
    ) -> (ReturnValue:bool)
   )
  ]}
 Is Sort Order Toggle Enabled {nodes [
   n1 = Is Sort Order Toggle Enabled!() -> (then, State:EMetaSoundMetadataEditorState Enum=NewEnumerator0) {is=title K2Node_FunctionEntry}
   > Return Node(ReturnValue=Select(NewEnumerator0=true, NewEnumerator1=true, NewEnumerator2=true, NewEnumerator3=true, Index=@n1))
  ]}
 On Reorder Clicked {nodes [On Reorder Clicked() {is=title K2Node_FunctionEntry}]}
]
variables [
 On Reorder Clicked {default=()}
 Widget State {type=EMetaSoundMetadataEditorState, default=Viewing Inputs}
]
widgets {
 info {namedSlotCount=2, rootWidgetClass=&VerticalBox, widgetCount=18}
 widgets [
  VerticalBox_1 {widgetClassPath=&VerticalBox, widget=VerticalBox_1}
  Border_0 {
   widgetClassPath=&Border
   parent=VerticalBox_1
   widget=Border_0
   properties {brushColor=<rgba 0.0284 0.0284 0.0284 1>, padding=<ltrb 8 4 8 4>}
   slot {class=VerticalBoxSlot}
  }
  Header_Box {
   widgetClassPath=&HorizontalBox
   bIsVariable=true
   parent=Border_0
   widget=Header_Box
   slot {class=BorderSlot, properties {padding=<ltrb 8 4 8 4>}}
  }
  Border_95 {
   widgetClassPath=&Border
   parent=Header_Box
   widget=Border_95
   properties {
    verticalAlignment=VAlign_Center
    background {
     bIsDynamicallyLoaded=false
     drawAs=RoundedBox
     imageType=NoImage
     mirroring=NoMirror
     resourceName=None
     resourceObject=null
     tiling=NoTile
     imageSize=<32 32>
     margin=<ltrb 0 0 0 0>
     outlineSettings {
      bUseBrushTransparency=false
      roundingType=FixedRadius
      width=0
      color {colorUseRule=UseColor_Specified, specifiedColor=<rgba 0 0 0 0>}
      cornerRadii=<4 4 4 4>
     }
     tintColor {colorUseRule=UseColor_Specified, specifiedColor=<rgba 1 1 1 1>}
     uVRegion {bIsValid=false, max=<0 0>, min=<0 0>}
    }
    brushColor=<rgba 0.0048 0.0048 0.0048 1>
    padding=<ltrb 2 2 2 2>
   }
   slot {class=HorizontalBoxSlot, properties {verticalAlignment=VAlign_Center}}
  }
  HorizontalBox_1 {
   widgetClassPath=&HorizontalBox
   parent=Border_95
   widget=HorizontalBox_1
   slot {class=BorderSlot, properties {verticalAlignment=VAlign_Center, padding=<ltrb 2 2 2 2>}}
  }
  Inputs_ToggleButton {
   widgetClassPath=/TechAudioToolsContent/Widgets/Editor/ToggleButtons/EUW_ToggleButton_ContentSlot
   bIsVariable=true
   parent=HorizontalBox_1
   widget=Inputs_ToggleButton
   properties {
    checked State=Checked
    check Box Style {
     checkBoxType=ToggleButton
     backgroundHoveredImage {
      bIsDynamicallyLoaded=false
      drawAs=RoundedBox
      imageType=NoImage
      mirroring=NoMirror
      resourceName=None
      resourceObject=null
      tiling=NoTile
      imageSize=<18 18>
      margin=<ltrb 0 0 0 0>
      outlineSettings {
       bUseBrushTransparency=false
       roundingType=FixedRadius
       width=1
       color {colorUseRule=UseColor_ColorTable, specifiedColor=<rgba 0 0 0 0>}
       cornerRadii=<3 3 3 3>
      }
      tintColor {colorUseRule=UseColor_Specified, specifiedColor=<rgba 0.0048 0.0048 0.0048 1>}
      uVRegion {bIsValid=false, max=<0 0>, min=<0 0>}
     }
     backgroundImage {
      bIsDynamicallyLoaded=false
      drawAs=RoundedBox
      imageType=NoImage
      mirroring=NoMirror
      resourceName=None
      resourceObject=null
      tiling=NoTile
      imageSize=<18 18>
      margin=<ltrb 0 0 0 0>
      outlineSettings {
       bUseBrushTransparency=false
       roundingType=FixedRadius
       width=1
       color {colorUseRule=UseColor_ColorTable, specifiedColor=<rgba 0 0 0 0>}
       cornerRadii=<3 3 3 3>
      }
      tintColor {colorUseRule=UseColor_Specified, specifiedColor=<rgba 0.0048 0.0048 0.0048 1>}
      uVRegion {bIsValid=false, max=<0 0>, min=<0 0>}
     }
     backgroundPressedImage {
      bIsDynamicallyLoaded=false
      drawAs=RoundedBox
      imageType=NoImage
      mirroring=NoMirror
      resourceName=None
      resourceObject=null
      tiling=NoTile
      imageSize=<18 18>
      margin=<ltrb 0 0 0 0>
      outlineSettings {
       bUseBrushTransparency=false
       roundingType=FixedRadius
       width=1
       color {colorUseRule=UseColor_ColorTable, specifiedColor=<rgba 0 0 0 0>}
       cornerRadii=<3 3 3 3>
      }
      tintColor {colorUseRule=UseColor_Specified, specifiedColor=<rgba 0.0048 0.0048 0.0048 1>}
      uVRegion {bIsValid=false, max=<0 0>, min=<0 0>}
     }
     borderBackgroundColor {colorUseRule=UseColor_Specified, specifiedColor=<rgba 1 1 1 1>}
     checkedForeground {colorUseRule=UseColor_Specified, specifiedColor=<rgba 1 1 1 1>}
     checkedHoveredForeground {colorUseRule=UseColor_Specified, specifiedColor=<rgba 1 1 1 1>}
     checkedHoveredImage {
      bIsDynamicallyLoaded=false
      drawAs=RoundedBox
      imageType=NoImage
      mirroring=NoMirror
      resourceName=None
      resourceObject=null
      tiling=NoTile
      imageSize=<32 32>
      margin=<ltrb 0 0 0 0>
      outlineSettings {
       bUseBrushTransparency=false
       roundingType=FixedRadius
       width=1
       color {colorUseRule=UseColor_ColorTable, specifiedColor=<rgba 0 0 0 0>}
       cornerRadii=<4 4 4 4>
      }
      tintColor {colorUseRule=UseColor_Specified, specifiedColor=<rgba 0.0395 0.0395 0.0395 1>}
      uVRegion {bIsValid=false, max=<0 0>, min=<0 0>}
     }
     checkedImage {
      bIsDynamicallyLoaded=false
      drawAs=RoundedBox
      imageType=NoImage
      mirroring=NoMirror
      resourceName=None
      resourceObject=null
      tiling=NoTile
      imageSize=<32 32>
      margin=<ltrb 0 0 0 0>
      outlineSettings {
       bUseBrushTransparency=false
       roundingType=FixedRadius
       width=1
       color {colorUseRule=UseColor_ColorTable, specifiedColor=<rgba 0 0 0 0>}
       cornerRadii=<4 4 4 4>
      }
      tintColor {colorUseRule=UseColor_Specified, specifiedColor=<rgba 0.0395 0.0395 0.0395 1>}
      uVRegion {bIsValid=false, max=<0 0>, min=<0 0>}
     }
     checkedPressedForeground {colorUseRule=UseColor_Specified, specifiedColor=<rgba 1 1 1 1>}
     checkedPressedImage {
      bIsDynamicallyLoaded=false
      drawAs=RoundedBox
      imageType=NoImage
      mirroring=NoMirror
      resourceName=None
      resourceObject=null
      tiling=NoTile
      imageSize=<32 32>
      margin=<ltrb 0 0 0 0>
      outlineSettings {
       bUseBrushTransparency=false
       roundingType=FixedRadius
       width=1
       color {colorUseRule=UseColor_ColorTable, specifiedColor=<rgba 0 0 0 0>}
       cornerRadii=<4 4 4 4>
      }
      tintColor {colorUseRule=UseColor_Specified, specifiedColor=<rgba 0.0176 0.0176 0.0176 1>}
      uVRegion {bIsValid=false, max=<0 0>, min=<0 0>}
     }
     checkedSlateSound {resourceObject=null}
     foregroundColor {colorUseRule=UseColor_Specified, specifiedColor=<rgba 1 1 1 1>}
     hoveredForeground {colorUseRule=UseColor_Specified, specifiedColor=<rgba 1 1 1 1>}
     hoveredSlateSound {resourceObject=null}
     padding=<ltrb 6 1.5 6 1.5>
     pressedForeground {colorUseRule=UseColor_Specified, specifiedColor=<rgba 1 1 1 1>}
     uncheckedHoveredImage {
      bIsDynamicallyLoaded=false
      drawAs=RoundedBox
      imageType=NoImage
      mirroring=NoMirror
      resourceName=None
      resourceObject=null
      tiling=NoTile
      imageSize=<32 32>
      margin=<ltrb 0 0 0 0>
      outlineSettings {
       bUseBrushTransparency=false
       roundingType=FixedRadius
       width=1
       color {colorUseRule=UseColor_ColorTable, specifiedColor=<rgba 0 0 0 0>}
       cornerRadii=<4 4 4 4>
      }
      tintColor {colorUseRule=UseColor_Specified, specifiedColor=<rgba 0.0395 0.0395 0.0395 1>}
      uVRegion {bIsValid=false, max=<0 0>, min=<0 0>}
     }
     uncheckedImage {
      bIsDynamicallyLoaded=false
      drawAs=NoDrawType
      imageType=NoImage
      mirroring=NoMirror
      resourceName=None
      resourceObject=null
      tiling=NoTile
      imageSize=<0 0>
      margin=<ltrb 0 0 0 0>
      outlineSettings {
       bUseBrushTransparency=false
       roundingType=HalfHeightRadius
       width=0
       color {colorUseRule=UseColor_Specified, specifiedColor=<rgba 0 0 0 0>}
       cornerRadii=<0 0 0 0>
      }
      tintColor {colorUseRule=UseColor_Specified, specifiedColor=<rgba 1 1 1 1>}
      uVRegion {bIsValid=false, max=<0 0>, min=<0 0>}
     }
     uncheckedPressedImage {
      bIsDynamicallyLoaded=false
      drawAs=RoundedBox
      imageType=NoImage
      mirroring=NoMirror
      resourceName=None
      resourceObject=null
      tiling=NoTile
      imageSize=<32 32>
      margin=<ltrb 0 0 0 0>
      outlineSettings {
       bUseBrushTransparency=false
       roundingType=FixedRadius
       width=1
       color {colorUseRule=UseColor_ColorTable, specifiedColor=<rgba 0 0 0 0>}
       cornerRadii=<4 4 4 4>
      }
      tintColor {colorUseRule=UseColor_Specified, specifiedColor=<rgba 0.0176 0.0176 0.0176 1>}
      uVRegion {bIsValid=false, max=<0 0>, min=<0 0>}
     }
     uncheckedSlateSound {resourceObject=null}
     undeterminedForeground {colorUseRule=UseColor_Specified, specifiedColor=<rgba 1 1 1 1>}
     undeterminedHoveredImage {
      bIsDynamicallyLoaded=false
      drawAs=NoDrawType
      imageType=NoImage
      mirroring=NoMirror
      resourceName=None
      resourceObject=null
      tiling=NoTile
      imageSize=<0 0>
      margin=<ltrb 0 0 0 0>
      outlineSettings {
       bUseBrushTransparency=false
       roundingType=HalfHeightRadius
       width=0
       color {colorUseRule=UseColor_Specified, specifiedColor=<rgba 0 0 0 0>}
       cornerRadii=<0 0 0 0>
      }
      tintColor {colorUseRule=UseColor_Specified, specifiedColor=<rgba 1 1 1 1>}
      uVRegion {bIsValid=false, max=<0 0>, min=<0 0>}
     }
     undeterminedImage {
      bIsDynamicallyLoaded=false
      drawAs=NoDrawType
      imageType=NoImage
      mirroring=NoMirror
      resourceName=None
      resourceObject=null
      tiling=NoTile
      imageSize=<0 0>
      margin=<ltrb 0 0 0 0>
      outlineSettings {
       bUseBrushTransparency=false
       roundingType=HalfHeightRadius
       width=0
       color {colorUseRule=UseColor_Specified, specifiedColor=<rgba 0 0 0 0>}
       cornerRadii=<0 0 0 0>
      }
      tintColor {colorUseRule=UseColor_Specified, specifiedColor=<rgba 1 1 1 1>}
      uVRegion {bIsValid=false, max=<0 0>, min=<0 0>}
     }
     undeterminedPressedImage {
      bIsDynamicallyLoaded=false
      drawAs=NoDrawType
      imageType=NoImage
      mirroring=NoMirror
      resourceName=None
      resourceObject=null
      tiling=NoTile
      imageSize=<0 0>
      margin=<ltrb 0 0 0 0>
      outlineSettings {
       bUseBrushTransparency=false
       roundingType=HalfHeightRadius
       width=0
       color {colorUseRule=UseColor_Specified, specifiedColor=<rgba 0 0 0 0>}
       cornerRadii=<0 0 0 0>
      }
      tintColor {colorUseRule=UseColor_Specified, specifiedColor=<rgba 1 1 1 1>}
      uVRegion {bIsValid=false, max=<0 0>, min=<0 0>}
     }
    }
   }
   slot {class=HorizontalBoxSlot}
  }
  Inputs {
   widgetClassPath=&TextBlock
   namedSlotHost=Inputs_ToggleButton
   widget=Inputs
   properties {
    text=Inputs
    textTransformPolicy=ToUpper
    colorAndOpacity {colorUseRule=UseColor_Specified, specifiedColor=<rgba 0.5776 0.5776 0.5776 1>}
    font {
     bForceMonospaced=false
     bMaterialIsStencil=false
     fontMaterial=null
     fontObject=&Roboto
     letterSpacing=0
     monospacedWidth=1
     size=8.25
     skewAmount=0
     typefaceFontName=Bold
     outlineSettings {
      bApplyOutlineToDropShadows=false
      bMiteredCorners=false
      bSeparateFillAlpha=false
      outlineMaterial=null
      outlineSize=0
      outlineColor=<rgba 0 0 0 1>
     }
    }
   }
  }
  Spacer_303 {widgetClassPath=&Spacer, parent=HorizontalBox_1, widget=Spacer_303, slot {class=HorizontalBoxSlot}}
  Outputs_ToggleButton {
   widgetClassPath=&EUW_ToggleButton_ContentSlot
   bIsVariable=true
   parent=HorizontalBox_1
   widget=Outputs_ToggleButton
   properties {check Box Style {
     checkBoxType=ToggleButton
     backgroundHoveredImage {
      bIsDynamicallyLoaded=false
      drawAs=RoundedBox
      imageType=NoImage
      mirroring=NoMirror
      resourceName=None
      resourceObject=null
      tiling=NoTile
      imageSize=<18 18>
      margin=<ltrb 0 0 0 0>
      outlineSettings {
       bUseBrushTransparency=false
       roundingType=FixedRadius
       width=1
       color {colorUseRule=UseColor_ColorTable, specifiedColor=<rgba 0 0 0 0>}
       cornerRadii=<3 3 3 3>
      }
      tintColor {colorUseRule=UseColor_Specified, specifiedColor=<rgba 0.0048 0.0048 0.0048 1>}
      uVRegion {bIsValid=false, max=<0 0>, min=<0 0>}
     }
     backgroundImage {
      bIsDynamicallyLoaded=false
      drawAs=RoundedBox
      imageType=NoImage
      mirroring=NoMirror
      resourceName=None
      resourceObject=null
      tiling=NoTile
      imageSize=<18 18>
      margin=<ltrb 0 0 0 0>
      outlineSettings {
       bUseBrushTransparency=false
       roundingType=FixedRadius
       width=1
       color {colorUseRule=UseColor_ColorTable, specifiedColor=<rgba 0 0 0 0>}
       cornerRadii=<3 3 3 3>
      }
      tintColor {colorUseRule=UseColor_Specified, specifiedColor=<rgba 0.0048 0.0048 0.0048 1>}
      uVRegion {bIsValid=false, max=<0 0>, min=<0 0>}
     }
     backgroundPressedImage {
      bIsDynamicallyLoaded=false
      drawAs=RoundedBox
      imageType=NoImage
      mirroring=NoMirror
      resourceName=None
      resourceObject=null
      tiling=NoTile
      imageSize=<18 18>
      margin=<ltrb 0 0 0 0>
      outlineSettings {
       bUseBrushTransparency=false
       roundingType=FixedRadius
       width=1
       color {colorUseRule=UseColor_ColorTable, specifiedColor=<rgba 0 0 0 0>}
       cornerRadii=<3 3 3 3>
      }
      tintColor {colorUseRule=UseColor_Specified, specifiedColor=<rgba 0.0048 0.0048 0.0048 1>}
      uVRegion {bIsValid=false, max=<0 0>, min=<0 0>}
     }
     borderBackgroundColor {colorUseRule=UseColor_Specified, specifiedColor=<rgba 1 1 1 1>}
     checkedForeground {colorUseRule=UseColor_Specified, specifiedColor=<rgba 1 1 1 1>}
     checkedHoveredForeground {colorUseRule=UseColor_Specified, specifiedColor=<rgba 1 1 1 1>}
     checkedHoveredImage {
      bIsDynamicallyLoaded=false
      drawAs=RoundedBox
      imageType=NoImage
      mirroring=NoMirror
      resourceName=None
      resourceObject=null
      tiling=NoTile
      imageSize=<32 32>
      margin=<ltrb 0 0 0 0>
      outlineSettings {
       bUseBrushTransparency=false
       roundingType=FixedRadius
       width=1
       color {colorUseRule=UseColor_ColorTable, specifiedColor=<rgba 0 0 0 0>}
       cornerRadii=<4 4 4 4>
      }
      tintColor {colorUseRule=UseColor_Specified, specifiedColor=<rgba 0.0395 0.0395 0.0395 1>}
      uVRegion {bIsValid=false, max=<0 0>, min=<0 0>}
     }
     checkedImage {
      bIsDynamicallyLoaded=false
      drawAs=RoundedBox
      imageType=NoImage
      mirroring=NoMirror
      resourceName=None
      resourceObject=null
      tiling=NoTile
      imageSize=<32 32>
      margin=<ltrb 0 0 0 0>
      outlineSettings {
       bUseBrushTransparency=false
       roundingType=FixedRadius
       width=1
       color {colorUseRule=UseColor_ColorTable, specifiedColor=<rgba 0 0 0 0>}
       cornerRadii=<4 4 4 4>
      }
      tintColor {colorUseRule=UseColor_Specified, specifiedColor=<rgba 0.0395 0.0395 0.0395 1>}
      uVRegion {bIsValid=false, max=<0 0>, min=<0 0>}
     }
     checkedPressedForeground {colorUseRule=UseColor_Specified, specifiedColor=<rgba 1 1 1 1>}
     checkedPressedImage {
      bIsDynamicallyLoaded=false
      drawAs=RoundedBox
      imageType=NoImage
      mirroring=NoMirror
      resourceName=None
      resourceObject=null
      tiling=NoTile
      imageSize=<32 32>
      margin=<ltrb 0 0 0 0>
      outlineSettings {
       bUseBrushTransparency=false
       roundingType=FixedRadius
       width=1
       color {colorUseRule=UseColor_ColorTable, specifiedColor=<rgba 0 0 0 0>}
       cornerRadii=<4 4 4 4>
      }
      tintColor {colorUseRule=UseColor_Specified, specifiedColor=<rgba 0.0176 0.0176 0.0176 1>}
      uVRegion {bIsValid=false, max=<0 0>, min=<0 0>}
     }
     checkedSlateSound {resourceObject=null}
     foregroundColor {colorUseRule=UseColor_Specified, specifiedColor=<rgba 1 1 1 1>}
     hoveredForeground {colorUseRule=UseColor_Specified, specifiedColor=<rgba 1 1 1 1>}
     hoveredSlateSound {resourceObject=null}
     padding=<ltrb 6 1.5 6 1.5>
     pressedForeground {colorUseRule=UseColor_Specified, specifiedColor=<rgba 1 1 1 1>}
     uncheckedHoveredImage {
      bIsDynamicallyLoaded=false
      drawAs=RoundedBox
      imageType=NoImage
      mirroring=NoMirror
      resourceName=None
      resourceObject=null
      tiling=NoTile
      imageSize=<32 32>
      margin=<ltrb 0 0 0 0>
      outlineSettings {
       bUseBrushTransparency=false
       roundingType=FixedRadius
       width=1
       color {colorUseRule=UseColor_ColorTable, specifiedColor=<rgba 0 0 0 0>}
       cornerRadii=<4 4 4 4>
      }
      tintColor {colorUseRule=UseColor_Specified, specifiedColor=<rgba 0.0395 0.0395 0.0395 1>}
      uVRegion {bIsValid=false, max=<0 0>, min=<0 0>}
     }
     uncheckedImage {
      bIsDynamicallyLoaded=false
      drawAs=NoDrawType
      imageType=NoImage
      mirroring=NoMirror
      resourceName=None
      resourceObject=null
      tiling=NoTile
      imageSize=<0 0>
      margin=<ltrb 0 0 0 0>
      outlineSettings {
       bUseBrushTransparency=false
       roundingType=HalfHeightRadius
       width=0
       color {colorUseRule=UseColor_Specified, specifiedColor=<rgba 0 0 0 0>}
       cornerRadii=<0 0 0 0>
      }
      tintColor {colorUseRule=UseColor_Specified, specifiedColor=<rgba 1 1 1 1>}
      uVRegion {bIsValid=false, max=<0 0>, min=<0 0>}
     }
     uncheckedPressedImage {
      bIsDynamicallyLoaded=false
      drawAs=RoundedBox
      imageType=NoImage
      mirroring=NoMirror
      resourceName=None
      resourceObject=null
      tiling=NoTile
      imageSize=<32 32>
      margin=<ltrb 0 0 0 0>
      outlineSettings {
       bUseBrushTransparency=false
       roundingType=FixedRadius
       width=1
       color {colorUseRule=UseColor_ColorTable, specifiedColor=<rgba 0 0 0 0>}
       cornerRadii=<4 4 4 4>
      }
      tintColor {colorUseRule=UseColor_Specified, specifiedColor=<rgba 0.0176 0.0176 0.0176 1>}
      uVRegion {bIsValid=false, max=<0 0>, min=<0 0>}
     }
     uncheckedSlateSound {resourceObject=null}
     undeterminedForeground {colorUseRule=UseColor_Specified, specifiedColor=<rgba 1 1 1 1>}
     undeterminedHoveredImage {
      bIsDynamicallyLoaded=false
      drawAs=NoDrawType
      imageType=NoImage
      mirroring=NoMirror
      resourceName=None
      resourceObject=null
      tiling=NoTile
      imageSize=<0 0>
      margin=<ltrb 0 0 0 0>
      outlineSettings {
       bUseBrushTransparency=false
       roundingType=HalfHeightRadius
       width=0
       color {colorUseRule=UseColor_Specified, specifiedColor=<rgba 0 0 0 0>}
       cornerRadii=<0 0 0 0>
      }
      tintColor {colorUseRule=UseColor_Specified, specifiedColor=<rgba 1 1 1 1>}
      uVRegion {bIsValid=false, max=<0 0>, min=<0 0>}
     }
     undeterminedImage {
      bIsDynamicallyLoaded=false
      drawAs=NoDrawType
      imageType=NoImage
      mirroring=NoMirror
      resourceName=None
      resourceObject=null
      tiling=NoTile
      imageSize=<0 0>
      margin=<ltrb 0 0 0 0>
      outlineSettings {
       bUseBrushTransparency=false
       roundingType=HalfHeightRadius
       width=0
       color {colorUseRule=UseColor_Specified, specifiedColor=<rgba 0 0 0 0>}
       cornerRadii=<0 0 0 0>
      }
      tintColor {colorUseRule=UseColor_Specified, specifiedColor=<rgba 1 1 1 1>}
      uVRegion {bIsValid=false, max=<0 0>, min=<0 0>}
     }
     undeterminedPressedImage {
      bIsDynamicallyLoaded=false
      drawAs=NoDrawType
      imageType=NoImage
      mirroring=NoMirror
      resourceName=None
      resourceObject=null
      tiling=NoTile
      imageSize=<0 0>
      margin=<ltrb 0 0 0 0>
      outlineSettings {
       bUseBrushTransparency=false
       roundingType=HalfHeightRadius
       width=0
       color {colorUseRule=UseColor_Specified, specifiedColor=<rgba 0 0 0 0>}
       cornerRadii=<0 0 0 0>
      }
      tintColor {colorUseRule=UseColor_Specified, specifiedColor=<rgba 1 1 1 1>}
      uVRegion {bIsValid=false, max=<0 0>, min=<0 0>}
     }
    }}
   slot {class=HorizontalBoxSlot}
  }
  Outputs {
   widgetClassPath=&TextBlock
   namedSlotHost=Outputs_ToggleButton
   widget=Outputs
   properties {
    text=Outputs
    textTransformPolicy=ToUpper
    colorAndOpacity {colorUseRule=UseColor_Specified, specifiedColor=<rgba 0.5776 0.5776 0.5776 1>}
    font {
     bForceMonospaced=false
     bMaterialIsStencil=false
     fontMaterial=null
     fontObject=&Roboto
     letterSpacing=0
     monospacedWidth=1
     size=8.25
     skewAmount=0
     typefaceFontName=Bold
     outlineSettings {
      bApplyOutlineToDropShadows=false
      bMiteredCorners=false
      bSeparateFillAlpha=false
      outlineMaterial=null
      outlineSize=0
      outlineColor=<rgba 0 0 0 1>
     }
    }
   }
  }
  Spacer_195 {
   widgetClassPath=&Spacer
   parent=Header_Box
   widget=Spacer_195
   slot {class=HorizontalBoxSlot, properties {size {sizeRule=Fill, value=1}}}
  }
  SortOrderEditor_ToggleButton {
   widgetClassPath=/TechAudioToolsContent/Widgets/Editor/Lists/MetaSoundMemberLists/EUW_MetaSoundMemberList_SearchToggleButton
   bIsVariable=true
   parent=Header_Box
   widget=SortOrderEditor_ToggleButton
   properties {icon Brush Path=EditorStyle.UndoHistory.TabIcon, unchecked Tool Tip=Edit Sort Order, default Icon Color=<rgba 0.3663 0.3663 0.3663 1>}
   slot {class=HorizontalBoxSlot, properties {verticalAlignment=VAlign_Center, padding=<ltrb 0 0 4 0>}}
  }
  Search_ToggleButton {
   widgetClassPath=&EUW_MetaSoundMemberList_SearchToggleButton
   bIsVariable=true
   parent=Header_Box
   widget=Search_ToggleButton
   properties {checked Tool Tip=Search Filter, unchecked Tool Tip=Search Filter}
   slot {class=HorizontalBoxSlot, properties {horizontalAlignment=HAlign_Center, verticalAlignment=VAlign_Center}}
  }
  WidgetSwitcher_1 {
   widgetClassPath=&WidgetSwitcher
   parent=VerticalBox_1
   widget=WidgetSwitcher_1
   properties {bOverride_Cursor=true}
   slot {class=VerticalBoxSlot, properties {size {sizeRule=Fill, value=1}}}
  }
  InputMetadata_ListView {
   widgetClassPath=&/EUW_MetaSoundMetadataEditor_InputList
   bIsVariable=true
   parent=WidgetSwitcher_1
   widget=InputMetadata_ListView
   properties {techAudioToolsListViewModel=&self:WidgetTree.InputMetadata_ListView.TechAudioToolsListViewModel_0}
   slot {class=WidgetSwitcherSlot}
  }
  InputSortOrderEditor_ListView {
   widgetClassPath=&/EUW_MetaSoundMetadataEditor_InputSortOrderEditorList
   bIsVariable=true
   parent=WidgetSwitcher_1
   widget=InputSortOrderEditor_ListView
   properties {techAudioToolsListViewModel=&self:WidgetTree.InputSortOrderEditor_ListView.TechAudioToolsListViewModel_0}
   slot {class=WidgetSwitcherSlot}
  }
  OutputMetadata_ListView {
   widgetClassPath=&/EUW_MetaSoundMetadataEditor_OutputList
   bIsVariable=true
   parent=WidgetSwitcher_1
   widget=OutputMetadata_ListView
   properties {techAudioToolsListViewModel=&self:WidgetTree.OutputMetadata_ListView.TechAudioToolsListViewModel_0}
   slot {class=WidgetSwitcherSlot}
  }
  OutputSortOrderEditor_ListView {
   widgetClassPath=&/EUW_MetaSoundMetadataEditor_OutputSortOrderEditorList
   bIsVariable=true
   parent=WidgetSwitcher_1
   widget=OutputSortOrderEditor_ListView
   properties {techAudioToolsListViewModel=&self:WidgetTree.OutputSortOrderEditor_ListView.TechAudioToolsListViewModel_0}
   slot {class=WidgetSwitcherSlot}
  }
 ]
}
